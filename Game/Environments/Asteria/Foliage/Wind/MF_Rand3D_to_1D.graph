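# Material function MF_Rand3D_to_1D in the Asteria foliage wind set that hashes a 3D seed position into a pseudo-random scalar.
# The Seed Position input, previewed with actor world position, is dotted with a constant vector, passed through Sine and scaled by a large constant.
# The fractional part becomes the Rand 1D output in the zero-to-one range, following the classic sine-fract shader hash.
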MaterialFunction /Game/Environments/Asteria/Foliage/Wind/MF_Rand3D_to_1D
graph {nodes [
  n1 = ActorPositionWS()
  n2 = Sine(None=DotProduct(
    A=Constant3Vector() {constant=<rgba 12.9898 4.1414 1.3204 0>}
    B=FunctionInput(Preview=@n1) {bUsePreviewValueAsDefault=true, inputName=Seed Position}
   ))
  n3 = Constant() {r=43758.5469}
  FunctionOutput(None=Frac(None=Multiply(A=@n2, B=@n3))) {outputName=Rand 1D}
 ]}
