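# Material function that supplies foliage shaders with wind intensity, direction, speed and a storm mask.
# Position-based UVs, varied by plant-type switches, pan across global wind textures to give intensity and noisy direction, while plain direction is constant.
# A weather mask takes the maximum of wind gusts, safe-zone storm proximity and vortex falloff, with debug overrides.
# Texture mip level is lerped between min and max parameters to vary the wind pattern.
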MaterialFunction /Game/Environments/Asteria/Foliage/Wind/MF_WindActorData
graph {nodes [
  CollectionParameter() {collection=&/MPC_Wind, parameterName=WindDirection}
  CollectionParameter() {collection=&FortniteMaterialParameters, parameterName=WindSpeed}
  Constant() {r=1}
  n1 = AppendVector(
   A=ComponentMask(None=NamedRerouteUsage()) {r=true}
   B=Divide(A=ScalarParameter() {group=Animation Medium Plant 2024, parameterName=Anim Style Selection 0-1})
  )
  n2 = TextureSample!(UVs=@n1, Tex, Level=NamedRerouteUsage(), Apply View MipBias) {
   mipValueMode=TMVM_MipLevel
   samplerType=SAMPLERTYPE_LinearColor
   texture=&/STATIC_WindGlobal_Medium_Foliage
  }
  n3 = TextureSample(UVs=NamedRerouteUsage(), Level=NamedRerouteUsage()) {
   mipValueMode=TMVM_MipLevel
   samplerSource=SSM_Wrap_WorldGroupSettings
   samplerType=SAMPLERTYPE_LinearColor
   texture=&/RT_WindGlobal_STATIC_S30_MIPWindSpeed
  }
  n4 = StaticSwitch(True=@n2.RGB, False=@n3.RGB, Value=NamedRerouteUsage())
  n5 = ComponentMask(None=@n4) {b=true}
  FunctionOutput(None=@n5) {outputName=WindIntesnity}
  FunctionOutput(None=ComponentMask(None=Constant3Vector() {constant=<rgba 1 0 0 0>}) {b=true, g=true, r=true}) {outputName=WindI_Direction, sortPriority=1}
  FunctionOutput(None=CollectionParameter() {collection=&FortniteMaterialParameters, parameterName=WindStrength}) {outputName=Speed 0-1, sortPriority=2}
  FunctionOutput(None=@n5) {outputName=WindIntensitySingleSample}
  n6 = Multiply(A=ComponentMask(None=@n4) {g=true, r=true}) {constB=2}
  FunctionOutput(None=AppendVector(A=Subtract(A=@n6), B=Constant())) {outputName=WindI_Direction_WithNoise, sortPriority=1}
  FunctionOutput(None=Saturate(None=NamedRerouteUsage())) {outputName=StormMask, sortPriority=50}
  n7 = Constant3Vector()
  NamedRerouteDeclaration(None=FunctionInput(Preview=Transform(None=@n7) {transformSourceType=TRANSFORMSOURCE_Instance}) {bUsePreviewValueAsDefault=true, inputName=Position}) {name=Position, nodeColor=<rgba 0.0781 0 1 1>}
  NamedRerouteDeclaration(None=FunctionInput(Preview=StaticBool()) {bUsePreviewValueAsDefault=true, inputName=IsGrass, inputType=FunctionInput_StaticBool, sortPriority=40}) {name=IsGrass, nodeColor=<rgba 0.2656 0 1 1>}
  n8 = MaterialFunctionCall(Float3=NamedRerouteUsage()) {materialFunction=&BreakOutFloat3Components}
  n9 = Multiply(A=VertexColor().R, B=ScalarParameter() {group=Animation, parameterName=Random Anim Offset Per Blade})
  n10 = AppendVector(A=@n9, B=@n9)
  n11 = NamedRerouteUsage()
  n12 = ComponentMask(None=@n11) {g=true, r=true}
  n13 = ComponentMask(None=@n11) {b=true}
  n14 = AppendVector(A=@n13, B=@n13)
  n15 = StaticSwitch(
   True=Add(A=Add(A=@n8.R, B=@n8.R), B=@n8.B)
   False=StaticSwitch(True=Add(A=@n10, B=@n12), False=Add(A=@n12, B=@n14), Value=NamedRerouteUsage())
   Value=NamedRerouteUsage()
  )
  n16 = StaticSwitch(
   True=ScalarParameter() {defaultValue=3000, group=Animation, parameterName=WindMaskScale_Trees, sliderMin=3}
   False=ScalarParameter() {defaultValue=3220.4573, group=Animation, parameterName=WindMaskScale, sliderMin=3}
   Value=NamedRerouteUsage()
  )
  n17 = ScalarParameter() {defaultValue=-0.1, group=Animation, parameterName=WindSpeed}
  n18 = Constant()
  NamedRerouteDeclaration(None=Panner(Coordinate=Divide(A=@n15, B=@n16), Speed=AppendVector(A=@n17, B=@n18)) {bFractionalPart=true}) {name=UVsamplelocation, nodeColor=<rgba 0.3359 1 0 1>}
  NamedRerouteDeclaration(None=LinearInterpolate(
    A=ScalarParameter() {group=Animation Medium Plant 2024, parameterName=WindSpeed_Medium2024_MIPLevel_Min}
    B=ScalarParameter() {defaultValue=1, group=Animation Medium Plant 2024, parameterName=WindSpeed_Medium2024_MIPLevel_Max}
    Alpha=NamedRerouteUsage()
   )) {name=MipMapLevelWindSpeed_Medium2024, nodeColor=<rgba 0.3359 1 0 1>}
  NamedRerouteDeclaration(None=FunctionInput(Preview=StaticBool()) {bUsePreviewValueAsDefault=true, inputName=IsTrees, inputType=FunctionInput_StaticBool, sortPriority=40}) {name=IsTrees, nodeColor=<rgba 0.0547 1 0 1>}
  NamedRerouteDeclaration(None=FunctionInput(Preview=StaticBool()) {bUsePreviewValueAsDefault=true, inputName=IsMediumPlant, inputType=FunctionInput_StaticBool, sortPriority=45}) {name=IsMediumPlant, nodeColor=<rgba 1 0 0.375 1>}
  NamedRerouteDeclaration(None=FunctionInput(Preview=StaticBool()) {bUsePreviewValueAsDefault=true, inputName=IsMediumPlant2ndLevel, inputType=FunctionInput_StaticBool, sortPriority=45}) {name=isMediumPlant2ndLevel, nodeColor=<rgba 0 0.6953 1 1>}
  NamedRerouteDeclaration(None=FunctionInput(Preview=StaticBool()) {bUsePreviewValueAsDefault=true, inputName=IsMediumPlant_2024Anim, inputType=FunctionInput_StaticBool, sortPriority=45}) {name=IsMediumPlant_2024Anim, nodeColor=<rgba 1 0 0.375 1>}
  n19 = MaterialFunctionCall!(FadeOutTIme, ResetFadeInTime, FadeInTime) {materialFunction=/Game/Environments/TechArt/WindGustSystem/MF_WindGustsSystem}
  n20 = StaticSwitch(
   True=Multiply(A=@n19.FadeInFadeOutNormalized, B=ScalarParameter() {defaultValue=0.75, parameterName=MaxWindGusts})
   False=@n19.FadeInFadeOutNormalized
   Value=NamedRerouteUsage()
  )
  n21 = Constant3Vector()
  n22 = ComponentMask(None=StaticSwitchParameter(
    True=TransformPosition(None=@n21) {transformSourceType=TRANSFORMPOSSOURCE_Instance, transformType=TRANSFORMPOSSOURCE_TranslatedWorld}
    False=NamedRerouteUsage()
   ) {parameterName=UseCameraRelativeOptimization}) {g=true, r=true}
  n23 = CollectionParameter() {collection=&FortniteMaterialParameters, parameterName=AthenaSafeZoneCenter}
  n24 = TransformPosition(None=@n23) {transformSourceType=TRANSFORMPOSSOURCE_World, transformType=TRANSFORMPOSSOURCE_TranslatedWorld}
  n25 = ComponentMask(None=StaticSwitchParameter(True=Constant3Vector(), False=StaticSwitchParameter(True=@n24, False=@n23) {parameterName=UseCameraRelativeOptimization}) {parameterName=DebugStorm}) {g=true, r=true}
  n26 = Frac(None=Multiply(A=Time(), B=ScalarParameter() {defaultValue=0.104, parameterName=TestRadiusSpeed}))
  n27 = OneMinus(None=@n26)
  n28 = LinearInterpolate(
   A=ScalarParameter() {defaultValue=22120.4531, parameterName=TestRadius}
   B=ScalarParameter() {parameterName=TestRadius2}
   Alpha=Multiply(A=Multiply(A=@n26, B=@n27)) {constB=4}
  )
  n29 = CollectionParameter() {collection=&FortniteMaterialParameters, parameterName=AthenaSafeZoneRadius}
  n30 = ScalarParameter() {defaultValue=1500, parameterName=RadiusFromStormWall, sliderMax=5000, sliderMin=100}
  n31 = Add(A=Subtract(A=Distance(A=@n22, B=@n25), B=StaticSwitchParameter(True=@n28, False=@n29) {group=Animation, parameterName=DebugStorm}), B=@n30)
  n32 = CollectionParameter() {collection=&FortniteMaterialParameters, parameterName=AthenaVortexCenter}
  n33 = Length(None=Subtract(A=@n22, B=ComponentMask(None=@n32) {g=true, r=true})) {is=class MaterialExpressionLength}
  n34 = ScalarParameter() {defaultValue=30000, parameterName=VortexRadius}
  n35 = OneMinus(None=Saturate(None=Divide(A=@n33, B=@n34)))
  n36 = CollectionParameter() {collection=&FortniteMaterialParameters, parameterName=AthenaVortexStrength}
  n37 = Multiply(A=ScalarParameter() {defaultValue=2, parameterName=VortexStrength}, B=@n36)
  n38 = Max(A=Saturate(None=Divide(A=@n31, B=@n30)), B=Saturate(None=Multiply(A=@n35, B=@n37)))
  NamedRerouteDeclaration(None=StaticSwitchParameter(True=NamedRerouteUsage(), False=Max(A=@n20, B=@n38)) {group=Animation, parameterName=DebugWindSpeed}) {name=WeatherChangeMask, nodeColor=<rgba 0.875 0 1 1>}
  n39 = Time()
  n40 = ScalarParameter() {defaultValue=0.2, parameterName=DebugSpeed}
  MaterialExpressionRemap_0 = MaterialXRemap!(None=Sine(None=Multiply(A=@n39, B=@n40)), InputLow, InputHigh, TargetLow, TargetHigh) {inputLowDefault=-1, targetHighDefault=2, targetLowDefault=-0.5}
  NamedRerouteDeclaration(None=Saturate(None=@MaterialExpressionRemap_0)) {name=DebugWind, nodeColor=<rgba 0 1 0.5547 1>}
  n41 = NamedRerouteUsage()
  n42 = NamedRerouteUsage()
  n43 = Constant3Vector()
  n44 = FunctionInput(Preview=TransformPosition(None=@n43) {transformType=TRANSFORMPOSSOURCE_World}) {bUsePreviewValueAsDefault=true, inputName=MediumPlant_BasePivotPostion, sortPriority=100}
  NamedRerouteDeclaration(None=ComponentMask(None=StaticSwitch(True=@n44, False=@n41, Value=@n42)) {g=true, r=true}) {name=Sample Location for Storm Mask, nodeColor=<rgba 1 0 0.3281 1>}
  n45 = ScalarParameter() {defaultValue=0.624, parameterName=WindSpeed_MIPLevel_Max_MediumPlant}
  n46 = ScalarParameter() {parameterName=WindSpeed_MIPLevel_Max}
  n47 = NamedRerouteUsage()
  NamedRerouteDeclaration(None=LinearInterpolate(
    A=ScalarParameter() {defaultValue=1, parameterName=WindSpeed_MIPLevel_Min}
    B=StaticSwitch(True=@n45, False=@n46, Value=@n47)
    Alpha=NamedRerouteUsage()
   )) {name=MipMapLevelWindSpeed, nodeColor=<rgba 0.3359 1 0 1>}
  TextureSample!(UVs=NamedRerouteUsage(), Tex, Apply View MipBias) {desc=Temp Assign RT for realtime wind controls here, texture=/Engine/EngineResources/DefaultTexture}
  TextureSample(UVs=@n1) {desc=Temp Assign RT for realtime wind controls here, texture=&DefaultTexture}
  Transform(None=@n43) {transformSourceType=TRANSFORMSOURCE_Instance}
  VectorParameter() {parameterName=VortexCenter, defaultValue=<rgba -11008 -1280 4540.2563 0>}
 ]}
