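# Material function that outputs a wind-bend World Position Offset for foliage in the Asteria environment.
# Time, offset by a random value seeded from position, drives sine and cosine sway, scaled by strength plus a random detail term.
# A power-curved mask of bounds-normalized distance from a reference point weights the sway.
# The displaced vertex is rescaled to its original pivot distance, so foliage bends without stretching.
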MaterialFunction /Game/Environments/Asteria/Foliage/Wind/MF_Wind
graph {nodes [
  n1 = WorldPosition()
  n2 = Subtract(A=NamedRerouteUsage(), B=@n1.XYZ)
  n3 = NamedRerouteUsage()
  n4 = Add(
   A=Multiply(A=Time(), B=ScalarParameter() {defaultValue=0.25, group=Animation WInd, parameterName=TimeScale})
   B=MaterialFunctionCall!(Seed Position=@n3) {materialFunction=&/MF_Rand3D_to_1D}.Rand 1D
  )
  n5 = Add(A=Sine(None=@n4)) {constB=0.75}
  n6 = Multiply(A=Cosine(None=Multiply(A=@n4) {constB=1.13})) {constB=0.618}
  n7 = Add(A=@n3, B=Constant3Vector() {constant=<rgba 17 53 101 0>})
  n8 = Add(
   A=Multiply(
    A=MaterialFunctionCall(X=@n5, Y=@n6) {materialFunction=&MakeFloat3}.Result
    B=ScalarParameter() {defaultValue=16, group=Animation WInd, parameterName=Strength}
   )
   B=Multiply(
    A=MaterialFunctionCall(Seed Position=@n7) {materialFunction=&MF_Rand3D_to_1D}.Rand 1D
    B=ScalarParameter() {group=Animation WInd, parameterName=detailNoise}
   )
  )
  n9 = NamedRerouteUsage()
  n10 = WorldPosition()
  n11 = Divide(
   A=Multiply(A=Subtract(A=@n9, B=@n10.XYZ), B=VectorParameter() {defaultValue=<rgba 1.5 1.5 1 1>}.RGB)
   B=Divide(A=MaterialFunctionCall() {materialFunction=&ObjectLocalBounds}.Local Bounds Max) {constB=1}
  )
  n12 = Power(
   Base=Saturate(None=MaterialFunctionCall(Vector 3=@n11) {materialFunction=/Engine/Functions/Engine_MaterialFunctions02/Utility/VectorLength}.V3 Length)
   Exp=ScalarParameter() {defaultValue=3, group=Animation WInd, parameterName=BendMaskPow}
  ) {constExponent=2.71}
  n13 = Subtract(A=NamedRerouteUsage(), B=Add(A=@n1.XYZ, B=Multiply(A=@n8, B=@n12)))
  n14 = Divide(
   A=MaterialFunctionCall(Vector 3=@n13) {materialFunction=&VectorLength}.V3 Length
   B=MaterialFunctionCall(Vector 3=@n2) {materialFunction=&VectorLength}.V3 Length
  )
  FunctionOutput(None=Subtract(A=@n2, B=Divide(A=@n13, B=@n14))) {outputName=Bend WPO}
  MaterialFunctionCall!(Position=NamedRerouteUsage(), IsTrees, IsGrass, IsMediumPlant, IsMediumPlant2ndLevel, MediumPlant_BasePivotPostion) {materialFunction=&/MF_WindActorData}
  NamedRerouteDeclaration(None=TransformPosition(None=Constant3Vector()) {transformSourceType=TRANSFORMPOSSOURCE_Instance, transformType=TRANSFORMPOSSOURCE_World}) {name=InstancePosition, nodeColor=<rgba 0 1 0.5781 1>}
 ]}
